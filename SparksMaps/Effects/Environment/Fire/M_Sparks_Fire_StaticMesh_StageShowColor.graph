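# Translucent unlit fire material for static meshes and Niagara particles, tinted by Sparks stage show primary color or a fallback color.
# Packed flame flipbook samples, offset half a cycle, combine with panning detail noise to drive emissive and opacity.
# Low and mobile quality substitute eroded, dissolving wispy noise.
# Depth, near-camera and distance fades plus time-of-day scaling shape visibility, while an atan2-based rotation yaws the mesh by camera direction.
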Material /SparksMaps/Effects/Environment/Fire/M_Sparks_Fire_StaticMesh_StageShowColor
graph {
 properties {
  bCastRayTracedShadows=false
  bUsedWithNiagaraMeshParticles=true
  bUsedWithNiagaraSprites=true
  blendMode=BLEND_Translucent
  pixelDepthOffsetMode=PDOM_Legacy
  shadingModel=MSM_Unlit
  usedShadingModels=Unlit
 }
 nodes [
  n1 = Constant2Vector() {g=8, r=8}
  n2 = Multiply(A=@n1.R, B=@n1.R)
  n3 = Divide(A=@n2)
  n4 = Time() {bOverride_Period=true, desc="`", period=1}
  Add(A=@n3, B=LinearInterpolate(Alpha=Sine(None=@n4)) {constA=-8, constB=8})
  n5 = TextureCoordinate()
  n6 = Frac(None=ObjectPositionWS())
  n7 = Add(A=ComponentMask(None=@n6) {r=true}, B=ComponentMask(None=@n6) {g=true})
  n8 = DynamicParameter() {parameterIndex=2, paramNames [Ignition Normal.X, Ignition Normal.Y, Ignition Normal.Z, Size]}
  n9 = NamedRerouteUsage()
  n10 = Add(
   A=Multiply(A=Add(A=@n5, B=@n7), B=@n8.Size)
   B=MaterialFunctionCall(X=@n9, Y=@n9) {materialFunction=&MakeFloat2}.Result
  )
  n11 = ScalarParameter() {defaultValue=1, parameterName=Detail TexScale, sliderMax=1}
  n12 = Time()
  n13 = Multiply(A=@n9) {constB=25}
  n14 = Constant()
  n15 = Multiply(A=@n8.Size, B=ScalarParameter() {defaultValue=1.25, parameterName=Detail PanSpeed, sliderMax=1})
  n16 = TextureSampleParameter2D(UVs=Panner(Coordinate=Multiply(A=@n10, B=@n11) {constB=0.4}, Time=Add(A=@n12, B=@n13), Speed=AppendVector(A=@n14, B=@n15)) {speedY=0.05}) {
   parameterName=FireTexture
   samplerType=SAMPLERTYPE_LinearGrayscale
   texture=/Game/Weapons/FORT_Melee/Pickaxe_Cube_Ninja_Male/FX/T_FluidFlameTile
  }
  n17 = ComponentMask(None=Multiply(
    A=DynamicParameter() {parameterIndex=1, paramNames [Sprite_RandomOffset, Sprite_TimeOffset, Detail_Intensity, ""]}.Detail_Intensity
    B=Power(Base=@n16.G) {constExponent=0.5}
   ) {constA=0.5}) {r=true}
  CurveAtlasRowParameter(CurveTime=@n17) {
   atlas=/Game/Curie/Effects/temp/Curves/CA_Curie_Fire_Visuals
   curve=/Game/Curie/Effects/temp/Curves/Curve_Fiery_Orange
   defaultValue=1
   parameterName=CurveAtlas_fire
  }
  n18 = StaticSwitchParameter(True=ObjectPositionWS(), False=ActorPositionWS()) {defaultValue=true, parameterName=True_Object_False_Actor}
  n19 = ComponentMask(None=@n18) {r=true}
  n20 = ComponentMask(None=@n18) {g=true}
  Frac(None=Add(A=Add(A=@n19, B=@n20), B=ComponentMask(None=@n18) {b=true}))
  LinearInterpolate(
   A=ScalarParameter() {defaultValue=1.5, parameterName=SineSpeedLow}
   B=ScalarParameter() {defaultValue=40, parameterName=PanSpeedHigh}
   Alpha=NamedRerouteUsage()
  ) {constA=0.5, constB=0.6}
  n21 = TextureObject() {texture=/Game/Effects/Fort_Effects/Textures/Fire/2DGasTorch_Flipbk_packed}
  n22 = Constant2Vector() {g=8, r=8}
  n23 = NamedRerouteUsage()
  n24 = Add(A=VertexColor().R, B=@n23)
  n25 = Add(A=@n23, B=Time() {period=4})
  n26 = ScalarParameter() {defaultValue=30, parameterName=PanSpeedLow}
  n27 = ScalarParameter() {defaultValue=40, parameterName=PanSpeedHigh}
  n28 = NamedRerouteUsage()
  MaterialFunctionCall!(
   Texture=@n21
   UVs
   SubImages=@n22
   Frame=Multiply(
    A=StaticSwitchParameter(True=@n24, False=@n25) {parameterName=UseVertexColorToOffset?}
    B=LinearInterpolate(A=@n26, B=@n27, Alpha=@n28) {constA=0.5, constB=0.6}
   ) {constB=0.5}
  ) {materialFunction=/Engine/Functions/Engine_MaterialFunctions03/Texturing/SubUV_Function}
  n29 = ScalarParameter() {parameterName="Hue Shift Percentage (S)"}
  n30 = NamedRerouteUsage()
  n31 = ScalarParameter() {defaultValue=25, parameterName=MaxEmissive_Distance}
  n32 = NamedRerouteUsage()
  n33 = NamedRerouteUsage()
  n34 = MaterialFunctionCall!(
   Low _ PC Low and Med=@n32
   "High _ XB1, PS4"=@n33
   "Epic _ PC High, XSX, PS5"=@n33
   --------
   "Mobile Low _ PC Perf, Android"=@n32
   Mobile Med _ Switch=@n32
  ) {materialFunction=/Game/Global/MaterialFunctions/TechArt/MF_Fortnite_QualitySwitch_Vector3}
  n35 = Multiply(
   A=Power(Base=@n34.Result, Exp=ScalarParameter() {defaultValue=2.5, parameterName=Flipbook Power})
   B=LinearInterpolate(
    A=ScalarParameter() {defaultValue=0.5, parameterName=Detail Min, sliderMax=3, sliderMin=-1}
    B=ScalarParameter() {defaultValue=2.5, parameterName=Detail Max, sliderMax=3, sliderMin=-1}
    Alpha=NamedRerouteUsage()
   ) {constA=0.8, constB=2}
  )
  n36 = QualitySwitch(Default=@n35, Low=@n34.Result, Epic=@n35) {desc=quality Opactiy}
  n37 = ScalarParameter() {defaultValue=1.508, group=Output Color, parameterName=Opacity Exponent, sliderMax=3, sortPriority=8}
  n38 = ScalarParameter() {defaultValue=3, group=Output Color, parameterName=Opacity Scale, sliderMax=3, sortPriority=8}
  n39 = Multiply(A=@n38, B=ScalarParameter() {defaultValue=0.5, parameterName=LOW_MOBILE_Opacity_Mult})
  n40 = Saturate(None=Multiply(
    A=Power(Base=@n36, Exp=@n37)
    B=MaterialFunctionCall(
     Low _ PC Low and Med=@n39
     "High _ XB1, PS4"=@n38
     "Epic _ PC High, XSX, PS5"=@n38
     "Mobile Low _ PC Perf, Android"=@n39
     Mobile Med _ Switch=@n39
    ) {materialFunction=&MF_Fortnite_QualitySwitch_Vector3}.Result
   ) {constB=3})
  n41 = LinearInterpolate(
   A=VectorParameter() {parameterName=Color, defaultValue=<rgba 1 0.044 0.3284 1>}.RGB
   B=MaterialFunctionCall!(CustomUPosition) {materialFunction=/SparksStageShow/Systems/StageManager/FXManager/MaterialFunctions/MF_MojoStageColors}.Primary
   Alpha=ScalarParameter() {defaultValue=1, parameterName=UseStageShowColors}
  )
  n42 = Multiply(A=@n41, B=@n35)
  n43 = LinearInterpolate(A=@n42, B=Divide(A=@n42, B=@n40), Alpha=ScalarParameter() {defaultValue=0.45, group=Output Color, parameterName=UnMult, sliderMax=1})
  n44 = NamedRerouteUsage()
  n45 = QualitySwitch(
   Default=@n43
   Low=Multiply(
    A=@n41
    B=MaterialFunctionCall(
     Low _ PC Low and Med=@n44
     "High _ XB1, PS4"=@n33
     "Epic _ PC High, XSX, PS5"=@n33
     "Mobile Low _ PC Perf, Android"=@n44
     Mobile Med _ Switch=@n44
    ) {materialFunction=&MF_Fortnite_QualitySwitch_Vector3}.Result
   )
   Epic=@n43
  ) {desc=quality RGB}
  n46 = ScalarParameter() {defaultValue=20, group=Output Color, parameterName=Final Color Brightness, sliderMax=50, sortPriority=9}
  n47 = Multiply(
   A=Clamp(None=Multiply(A=@n30, B=@n31) {constB=2}) {maxDefault=2, minDefault=0.15}
   B=Multiply(A=Constant() {r=1}, B=Multiply(A=@n45, B=@n46))
  )
  MaterialFunctionCall(Emissive=Desaturation(
    None=MaterialFunctionCall(Hue Shift Percentage=@n29, Texture=@n47) {materialFunction=&HueShift}.Result
    Fraction=ScalarParameter() {parameterName=Desatutration}
   )) {materialFunction=&MF_FX_EmissiveAdjustments}
  n48 = Time() {period=4}
  n49 = Add(A=@n48, B=VertexColor().R)
  n50 = Add(A=NamedRerouteUsage(), B=@n48)
  n51 = ScalarParameter() {defaultValue=24, parameterName=PanSpeedLow_2ndTex}
  n52 = ScalarParameter() {defaultValue=37, parameterName=PanSpeedHigh_2ndTex}
  n53 = NamedRerouteUsage()
  MaterialFunctionCall(
   Texture=@n21
   SubImages=@n22
   Frame=Multiply(
    A=StaticSwitchParameter(True=@n49, False=@n50) {parameterName=UseVertexColorToOffset?}
    B=LinearInterpolate(A=@n51, B=@n52, Alpha=@n53) {constA=0.5, constB=0.6}
   ) {constB=0.5}
  ) {materialFunction=&SubUV_Function}
  n54 = DepthFade(Opacity=@n40, FadeDistance=ScalarParameter() {defaultValue=55, parameterName=DepthFadeDistance, sliderMax=300})
  n55 = QualitySwitch(Default=@n54, Low=@n40, Epic=@n54) {desc=quality Opactiy}
  n56 = MaterialFunctionCall(
   Value_Morning=ScalarParameter() {defaultValue=1, parameterName="Value_Morning (S)"}
   Value_Day=ScalarParameter() {defaultValue=1, parameterName="Value_Day (S)"}
   Value_Evening=ScalarParameter() {defaultValue=1, parameterName="Value_Evening (S)"}
   Value_Night=ScalarParameter() {defaultValue=2, parameterName="Value_Night (S)"}
  ) {materialFunction=&Scalar_Time_of_Day}
  n57 = ScalarParameter() {defaultValue=800, parameterName=NearFade}
  n58 = ScalarParameter() {defaultValue=250, parameterName=NearFadeGone}
  Multiply(
   A=Multiply(A=NamedRerouteUsage(), B=Multiply(A=@n55, B=@n56.Result))
   B=MaterialFunctionCall(Distance=Subtract(A=@n57, B=@n58)) {materialFunction=&Near_Camera_Fade}.Result
  )
  NamedRerouteDeclaration(None=MaterialFunctionCall(
    Fade_Start_Distance=ScalarParameter() {defaultValue=25000, parameterName="Fade_Start_Distance (S)"}
    Fade_End_Distance=ScalarParameter() {parameterName="Fade_End_Distance (S)"}
   ) {materialFunction=&MF_Distance_SyncedFade_FX}.Result Inverted) {name=FadeDistanceUseThis, nodeColor=<rgba 0.5373 1 0.7595 1>}
  n59 = TextureObject() {texture=&2DGasTorch_Flipbk_packed}
  n60 = Multiply(
   A=Add(A=NamedRerouteUsage(), B=Time() {period=4})
   B=LinearInterpolate(
    A=ScalarParameter() {defaultValue=30, parameterName=PanSpeedLow}
    B=ScalarParameter() {defaultValue=40, parameterName=PanSpeedHigh}
    Alpha=NamedRerouteUsage()
   ) {constA=0.5, constB=0.6}
  ) {constB=0.5}
  n61 = MaterialFunctionCall(Texture=@n59, SubImages=@n1, Frame=@n60) {materialFunction=&SubUV_Function}
  n62 = ComponentMask(None=@n61.RGB) {r=true}
  n63 = MaterialFunctionCall(Texture=@n59, SubImages=@n1, Frame=Modulo!(A=Add(A=@n60, B=@n3), B=@n2) {desc="(Frame + 32) % 64"}) {materialFunction=&SubUV_Function}
  n64 = ComponentMask(None=@n63.RGB) {r=true}
  n65 = ComponentMask(None=@n61.RGB) {g=true}
  n66 = ComponentMask(None=@n63.RGB) {g=true}
  n67 = ComponentMask(None=@n61.RGB) {b=true}
  n68 = ComponentMask(None=@n63.RGB) {b=true}
  n69 = Add(A=Add(A=Add(A=@n62, B=@n64), B=Add(A=@n65, B=@n66)), B=Add(A=@n67, B=@n68))
  n70 = Add(A=@n62, B=@n66)
  n71 = Add(A=@n65, B=@n68)
  n72 = Add(A=NamedRerouteUsage(), B=VertexColor().G)
  n73 = NamedRerouteUsage()
  NamedRerouteDeclaration(None=LinearInterpolate(A=Add(A=@n69, B=@n67), B=Add(A=@n70, B=@n71), Alpha=StaticSwitchParameter(True=@n72, False=@n73) {parameterName=UseVertexColorToOffset?})) {name=Flipbook Flame Texture, nodeColor=<rgba 1 0 0.7211 1>}
  n74 = Constant() {r=0.5}
  n75 = CameraPositionWS()
  n76 = Constant3Vector() {constant=<rgba 1 1 0 0>}
  MaterialExpressionNormalize_0 = Normalize(VectorInput=Subtract(A=ObjectPositionWS(), B=Multiply(A=@n75, B=@n76)))
  n77 = Custom!(In=@MaterialExpressionNormalize_0) {
   code="float2 output; \r\nfloat2 Input1; \r\noutput = atan2 (In.y,In.x); \r\nreturn (output);"
   outputType=CMOT_Float2
   inputs [{inputName=In, input {expression=&self:MaterialExpressionNormalize_0}}]
  }
  n78 = Divide(A=Divide(A=@n77, B=Constant() {r=3.142}), B=Constant() {r=2})
  n79 = AppendVector(A=TextureCoordinate(), B=ComponentMask(None=TextureCoordinate()) {r=true})
  NamedRerouteDeclaration(None=RotateAboutAxis(
    NormalizedRotationAxis=ObjectOrientation()
    RotationAngle=Add(A=@n74, B=@n78)
    PivotPoint=TransformPosition(None=@n79) {transformType=TRANSFORMPOSSOURCE_World}
    Position=WorldPosition().XYZ
   )) {name=WPO, nodeColor=<rgba 1 0 0.3281 1>}
  NamedRerouteDeclaration(None=@n17) {name=Tiling Flame Texture, nodeColor=<rgba 1 0 0.3516 1>}
  n80 = Multiply(A=TruncateLWC!(None=ObjectPositionWS()), B=ScalarParameter() {defaultValue=0.02, parameterName=TestScale})
  NamedRerouteDeclaration(None=ComponentMask(None=Frac(None=@n80)) {g=true}) {name=ObjPosRandom, nodeColor=<rgba 0.5937 1 0 1>}
  n81 = ScalarParameter() {defaultValue=0.6, parameterName=Time Scale}
  n82 = Add(A=ObjectPositionWS()) {constB=324.3412}
  n83 = Subtract(A=Subtract(A=ComponentMask(None=@n82) {r=true}, B=ComponentMask(None=@n82) {g=true}), B=ComponentMask(None=@n82) {b=true})
  n84 = NamedRerouteDeclaration(None=LinearInterpolate(
    A=ScalarParameter() {defaultValue=0.8, parameterName=Object Pos Frac Min}
    B=ScalarParameter() {defaultValue=1.1, parameterName=Object Pos Frac Max}
    Alpha=Frac(None=@n83)
   )) {name=ObjPosRandom_1, nodeColor=<rgba 0.5938 1 0 1>}
  NamedRerouteDeclaration(None=Multiply(A=Time(), B=Multiply(A=@n81, B=@n84))) {name=Time Scaled, nodeColor=<rgba 1 0.7266 0 1>}
  n85 = VectorParameter() {parameterName=UV Scale and Speed, defaultValue=<rgba 1 0.25 -0.2 0.66>}
  n86 = Multiply(A=TextureCoordinate(), B=ComponentMask(None=@n85.RGB) {g=true, r=true})
  n87 = NamedRerouteUsage()
  n88 = AppendVector(A=@n85.B, B=@n85.B)
  n89 = VectorParameter() {parameterName=UV Scale and Speed B, defaultValue=<rgba 2 0.5 0.3 1>}
  n90 = Multiply(A=TextureCoordinate(), B=ComponentMask(None=@n89.RGB) {g=true, r=true})
  n91 = AppendVector(A=@n89.B, B=@n89.B)
  n92 = Add(
   A=TextureSample(UVs=Panner(Coordinate=@n86, Time=@n87, Speed=@n88)) {samplerType=SAMPLERTYPE_Grayscale, texture=/Game/Effects/Fort_Effects/Textures/Fire/T_WispyFireNoise_Soft_M}.G
   B=TextureSample(UVs=Panner(Coordinate=@n90, Time=@n87, Speed=@n91)) {samplerType=SAMPLERTYPE_Grayscale, texture=&T_WispyFireNoise_Soft_M}.G
  )
  n93 = Saturate(None=ScalarParameter() {defaultValue=1, parameterName=Opacity})
  n94 = Saturate(None=Subtract(A=Divide(A=@n92), B=OneMinus(None=@n93)))
  n95 = ScalarParameter() {defaultValue=0.1, parameterName=asf}
  n96 = MaterialFunctionCall!(
   Gradient=SphereMask(A=TextureCoordinate() {uTiling=2}, B=Constant2Vector() {g=0.5, r=1}) {attenuationRadius=0.5, hardnessPercent=0}
   Erosion Amount=OneMinus(None=Power(Base=@n94, Exp=@n95))
  ) {materialFunction=/Game/Effects/Fort_Effects/MaterialFunctions/MF_ErosionNormalized}
  n97 = Add(A=NamedRerouteUsage(), B=LinearInterpolate(Alpha=VertexColor().G) {constA=0.6, constB=0.4})
  n98 = NamedRerouteUsage()
  n99 = Panner(
   Coordinate=TextureCoordinate() {uTiling=2}
   Speed=Multiply(A=StaticSwitchParameter(True=@n97, False=@n98) {parameterName=UseVertexColorToOffset?}, B=Constant2Vector() {g=2, r=0.5})
  ) {speedY=1}
  n100 = ComponentMask(None=TextureCoordinate()) {g=true}
  n101 = LinearInterpolate(
   A=ScalarParameter() {parameterName=Dissolve mask Min}
   B=ScalarParameter() {defaultValue=0.5, parameterName=Dissolve mask Max}
   Alpha=Power(Base=OneMinus(None=@n100)) {constExponent=1.3}
  ) {constA=0.1, constB=2}
  n102 = LinearInterpolate(
   A=ScalarParameter() {defaultValue=-1, parameterName=Disappear min}
   B=ScalarParameter() {defaultValue=2, parameterName=Disappear max}
   Alpha=Power(Base=TextureSample(UVs=@n99) {texture=/Game/Effects/Fort_Effects/Textures/Tile/T_ShieldCaustics_Tile_Chromatic}.R, Exp=Saturate(None=@n101)) {constExponent=0.2}
  ) {constA=-2, constB=2}
  NamedRerouteDeclaration(None=Multiply(A=@n96.Result, B=Saturate(None=@n102))) {name=LOW_MOBILE_Opacity, nodeColor=<rgba 0 0.2266 1 1>}
  NamedRerouteDeclaration(None=@n96.Result) {name=LOW_MOBILE_EMMISIVE, nodeColor=<rgba 1 0 0.5156 1>}
  NamedRerouteUsage()
  NamedRerouteUsage()
  Saturate(None=Divide(A=@n93, B=ScalarParameter() {defaultValue=0.5, parameterName=Smoke Fade Size}) {constB=0.5})
 ]
}
